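# Default Lit material for the Creative basketball toy, giving it an iridescent surface with configurable base, seam and "F" colors.
# View-angle dot product, time panning and a per-section offset texture drive a hue shift, desaturated to sample a curve atlas for the iridescent tint.
# An optional tiled pattern appears where the reflection vector aligns with the sun/moon direction.
# A mask texture layers seam and F colors, plus optional half-ball iridescence.
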Material /BRCosmetics/Creative/Toys/Basketball/Materials/M_CP_Basketball_Iridescence_01
graph {
 properties {pixelDepthOffsetMode=PDOM_Legacy, usedShadingModels=Default Lit}
 nodes [
  n1 = PixelNormalWS()
  n2 = TextureSample() {texture=/BRCosmetics/Creative/Toys/Basketball/Textures/SportingGoods_Basketball_Creative_SectionOffset}
  n3 = ComponentMask(None=LinearInterpolate(B=Multiply(A=@n2.RGB) {constB=0.5}, Alpha=ScalarParameter() {group=Hue Shift, parameterName=Section Offset})) {g=true, r=true}
  Add(
   A=Multiply(A=ComponentMask(None=@n1) {g=true, r=true}, B=ScalarParameter() {defaultValue=1, group=Hue Shift, parameterName=Hue Shift Tiling})
   B=@n3
  )
  n4 = DotProduct(A=PixelNormalWS(), B=CameraVectorWS())
  AppendVector(A=@n4, B=@n4)
  n5 = CollectionParameter() {
   collection=&FortniteMaterialParameters
   parameterName=SunAndMoonModelDirectionalVector
  }
  n6 = DotProduct(A=ComponentMask(None=@n5) {b=true, g=true, r=true}, B=ReflectionVectorWS())
  n7 = Multiply(
   A=Power(Base=Clamp(None=@n6), Exp=ScalarParameter() {defaultValue=4, group=Pattern, parameterName="Pattern Mask Power (Spec Dir)", sortPriority=30}) {constExponent=20}
   B=ScalarParameter() {defaultValue=32, group=Pattern, parameterName=Pattern Intensity, sortPriority=60}
  )
  Clamp(None=@n7)
  n8 = MaterialFunctionCall(TextureObject=TextureObject() {texture=/Game/Characters/Player/Female/Medium/Bodies/F_MED_Neon_Cat_Spy/Textures/Gradient_White}) {materialFunction=/Game/Weapons/WeaponSkins/Wraps/MaterialFunctions/WorldAlignedTexture}
  n9 = ComponentMask(None=WorldPosition().XYZ) {b=true}
  LinearInterpolate(A=@n8.XY Texture, B=@n8.XY Texture, Alpha=Clamp(None=Ceil(None=@n9)))
  n10 = Time()
  n11 = ScalarParameter() {group=Hue Shift, parameterName=Pan Speed}
  n12 = Add(A=Add(A=@n4, B=Multiply(A=@n10, B=@n11)), B=@n3)
  n13 = VectorParameter() {group=Hue Shift, parameterName=Hue Shift Values, sortPriority=3, defaultValue=<rgba 1 1 1 1>}
  n14 = CurveAtlasRowParameter(CurveTime=Desaturation(None=MaterialFunctionCall(Hue Shift Percentage=@n12, Texture=@n13.RGB) {materialFunction=&HueShift}.Result)) {
   atlas=&AthenaCurveAtlas01
   curve=/Game/Athena/Environments/Nexus/Materials/FXshader/Curve_Aix
   defaultValue=77
   group=Hue Shift
   parameterName=Curve
   sortPriority=2
  }
  n15 = ScalarParameter() {defaultValue=1, group=Hue Shift, parameterName=Hue Shift Brightness, sortPriority=5}
  n16 = VectorParameter() {group=Color, parameterName=Iridescence Base Color, sortPriority=2, defaultValue=<rgba 1 1 1 1>}
  n17 = TextureSampleParameter2D(UVs=Multiply(A=TextureCoordinate(), B=ScalarParameter() {defaultValue=1, group=Pattern, parameterName=Pattern Tiling, sortPriority=21})) {
   group=Pattern
   parameterName=Pattern Texture
   sortPriority=1
   texture=/BRCosmetics/Creative/Toys/Basketball/Textures/SportingGoods_Basketball_Creative_Pattern_TriangleBars_01
  }
  n18 = LinearInterpolate(
   A=@n16.RGB
   B=VectorParameter() {group=Color, parameterName=Iridescence Pattern Color, sortPriority=2, defaultValue=<rgba 1 1 1 1>}.RGB
   Alpha=StaticSwitchParameter(True=OneMinus(None=@n17.RGB), False=@n17.RGB) {group=Pattern, parameterName=Invert Pattern, sortPriority=21}
  )
  n19 = Multiply(
   A=LinearInterpolate(B=Multiply(A=@n14, B=@n15), Alpha=ScalarParameter() {defaultValue=1, group=Hue Shift, parameterName=Hue Shift Intensity, sortPriority=4}) {constA=1}
   B=StaticSwitchParameter(True=LinearInterpolate(A=@n16.RGB, B=@n18, Alpha=@n7), False=@n16.RGB) {defaultValue=true, group=Pattern, parameterName=Use Iridescence Pattern, sortPriority=0}
  )
  n20 = TextureSample() {texture=/BRCosmetics/Creative/Toys/Basketball/Textures/SportingGoods_Basketball_Creative_Masks_01}
  n21 = LinearInterpolate(
   A=@n19
   B=VectorParameter() {group=Color, parameterName=Base Color, sortPriority=2, defaultValue=<rgba 1 1 1 1>}.RGB
   Alpha=StaticSwitchParameter(True=OneMinus(None=@n20.B), False=@n20.B) {parameterName=Flip Half Iridescence Mask, sortPriority=2}
  )
  LinearInterpolate(
   A=LinearInterpolate(
    A=StaticSwitchParameter(True=@n21, False=@n19) {parameterName=Half Iridescence, sortPriority=1}
    B=VectorParameter() {group=Color, parameterName=Seam Color, sortPriority=2, defaultValue=<rgba 1 1 1 1>}.RGB
    Alpha=@n20.G
   )
   B=VectorParameter() {group=Color, parameterName=F Color, sortPriority=2, defaultValue=<rgba 1 1 1 1>}.RGB
   Alpha=@n20.R
  )
  Multiply(A=@n8.XY Texture) {constB=-1}
  TextureCoordinate()
  TextureSample() {samplerType=SAMPLERTYPE_Normal, texture=/BRCosmetics/Creative/Toys/Basketball/Textures/SportingGoods_Basketball_Creative_01_N}
  TextureSample() {texture=/BRCosmetics/Creative/Toys/Basketball/Textures/SportingGoods_Basketball_Creative_01_S}
 ]
}
